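# Sound cue for the bubble foley of the Blowing Bubbles emote in Athena (Battle Royale).
# A mixer layers one of three random bubble foley waves with one of two random inhale waves, the inhale at reduced volume.
# It uses emote attenuation, the emote foley sound class and foley concurrency, and sends to a music audio-analysis submix.
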SoundCue /BRCosmetics/Sounds/Emotes/BlowingBubbles/Athena_Emote_BlowingBubbles_Bubble_Foley_Cue
root=Mixer_0
nodes [
 {id=Mixer_0, children [Random_0,Random_1], properties {inputVolume [1,0.35]}}
 {id=Random_0, children [WavePlayer_0,WavePlayer_2,WavePlayer_3], properties {weights [1,1,1]}}
 {id=Random_1, children [WavePlayer_1,WavePlayer_4], properties {weights [1,1]}}
 {id=WavePlayer_0, properties {soundWaveAssetPtr=/Game/Athena/Sounds/Emotes/BlowingBubbles/Athena_Emote_BlowingBubbles_BubbleFoley_01}}
 {id=WavePlayer_1, properties {soundWaveAssetPtr=/Game/Athena/Sounds/Emotes/BlowingBubbles/Athena_Emote_BlowingBubbles_Inhale_01}}
 {id=WavePlayer_2, properties {soundWaveAssetPtr=/Game/Athena/Sounds/Emotes/BlowingBubbles/Athena_Emote_BlowingBubbles_BubbleFoley_02}}
 {id=WavePlayer_3, properties {soundWaveAssetPtr=/Game/Athena/Sounds/Emotes/BlowingBubbles/Athena_Emote_BlowingBubbles_BubbleFoley_03}}
 {id=WavePlayer_4, properties {soundWaveAssetPtr=/Game/Athena/Sounds/Emotes/BlowingBubbles/Athena_Emote_BlowingBubbles_Inhale_02}}
]
properties {
 attenuationSettings=&Fort_Emotes_Athena_Att
 audioPropertiesSheet=/BRCosmetics/Sounds/Settings/PropertySheets/Emotes/APS_Cosmetic_Emote_Foley
 maxDistance=1600
 soundClassObject=/Game/Sounds/FortSoundClassesAndModes/Fort_Foley_Emote
 volumeMultiplier=1
 concurrencyOverrides {resolutionRule=PreventNew}
 concurrencySet [/Game/Sounds/Concurrency_Rules/Athena_EmoteFoley_Conc]
 soundSubmixSends [{minSendLevel=0, soundSubmix=&Fortnite_AudioAnalysis_Music}]
}
